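# Physics asset for the female F_MED_SunBeam backpack, defining capsule bodies and joints for its two dangling dynamic chains.
# Left and right chains each run seven segments from a small master body on the pack, with moderate swing limits and tight twist limits between links.
# Capsules on root, pelvis, spine, arms and legs approximate the wearer's body, with free angular joints, though foot_l has no constraint.
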PhysicsAsset /BRCosmetics/Accessories/FORT_Backpacks/Backpack_F_MED_SunBeam/Meshes/F_MED_SunBeam_Pack_Female_Physics
bodies {
 calf_l [{shapeType=Capsule, length=34, radius=10, shapeName=calf_r_capsule, center=<-23.3518 2 0>, rotation=<rot 90 0 0>}]
 calf_r [{shapeType=Capsule, length=34, radius=10, shapeName=calf_r_capsule, center=<23.3518 -2 0>, rotation=<rot 90 0 0>}]
 dyn_chain_01_l [{shapeType=Capsule, length=5, radius=2, shapeName=dyn_chain_01_l_capsule, center=<4 0 0>, rotation=<rot 90 0 0>}]
 dyn_chain_01_l_MSTR [{shapeType=Capsule, length=0.505, radius=0.505, shapeName=dyn_chain_01_l_MSTR_capsule, rotation=<rot 0 -90 0>}]
 dyn_chain_01_r [{shapeType=Capsule, length=5, radius=2, shapeName=dyn_chain_01_r_capsule, center=<4 0 0>, rotation=<rot 90 0 0>}]
 dyn_chain_01_r_MSTR [{shapeType=Capsule, length=0.505, radius=0.505, shapeName=dyn_chain_01_r_MSTR_capsule, rotation=<rot 0 -90 0>}]
 dyn_chain_02_l [{shapeType=Capsule, length=4.5, radius=2, shapeName=dyn_chain_02_l_capsule, center=<4 0 0>, rotation=<rot 90 0 0>}]
 dyn_chain_02_r [{shapeType=Capsule, length=4.5, radius=2, shapeName=dyn_chain_02_r_capsule, center=<4 0 0>, rotation=<rot 90 0 0>}]
 dyn_chain_03_l [{shapeType=Capsule, length=5, radius=2, shapeName=dyn_chain_03_l_capsule, center=<4.5 0 0>, rotation=<rot 90 0 0>}]
 dyn_chain_03_r [{shapeType=Capsule, length=5, radius=2, shapeName=dyn_chain_03_r_capsule, center=<4.5 0 0>, rotation=<rot 90 0 0>}]
 dyn_chain_04_l [{shapeType=Capsule, length=5.5, radius=2, shapeName=dyn_chain_04_l_capsule, center=<5 0 0>, rotation=<rot 90 0 0>}]
 dyn_chain_04_r [{shapeType=Capsule, length=5.5, radius=2, shapeName=dyn_chain_04_r_capsule, center=<5 0 0>, rotation=<rot 90 0 0>}]
 dyn_chain_05_l [{shapeType=Capsule, length=5.5, radius=2, shapeName=dyn_chain_05_l_capsule, center=<5 0 0>, rotation=<rot 90 0 0>}]
 dyn_chain_05_r [{shapeType=Capsule, length=5.5, radius=2, shapeName=dyn_chain_05_r_capsule, center=<5 0 0>, rotation=<rot 90 0 0>}]
 dyn_chain_06_l [{shapeType=Capsule, length=6, radius=2, shapeName=dyn_chain_06_l_capsule, center=<5 0 0>, rotation=<rot 90 0 0>}]
 dyn_chain_06_r [{shapeType=Capsule, length=6, radius=2, shapeName=dyn_chain_06_r_capsule, center=<5 0 0>, rotation=<rot 90 0 0>}]
 dyn_chain_07_l [{shapeType=Capsule, length=10.5, radius=2, shapeName=dyn_chain_07_l_capsule, center=<7.5353 0 0>, rotation=<rot 90 0 0>}]
 dyn_chain_07_r [{shapeType=Capsule, length=10.5, radius=2, shapeName=dyn_chain_07_r_capsule, center=<7.5351 0 0>, rotation=<rot 90 0 0>}]
 foot_l [{
   shapeType=Capsule
   length=11.1318
   radius=5.752
   shapeName=foot_l_capsule
   center=<-3.3322 5.4999 -0.0382>
   rotation=<rot 89.6015 89.9826 179.9998>
  }]
 foot_r [{
   shapeType=Capsule
   length=11.1318
   radius=5.752
   shapeName=foot_r_capsule
   center=<3.3322 -5.4999 0.0382>
   rotation=<rot 89.6015 89.9826 -0.0002>
  }]
 lowerarm_l [{shapeType=Capsule, length=20.5, radius=4.5, shapeName=lowerarm_l_capsule, center=<11.3312 0 0>, rotation=<rot 90 0 0>}]
 lowerarm_r [{shapeType=Capsule, length=20.5, radius=4.5, shapeName=lowerarm_l_capsule, center=<-11.3312 0 0>, rotation=<rot 90 0 0>}]
 pack [{
   shapeType=Capsule
   length=12.2691
   radius=5.0927
   shapeName=pack_capsule
   center=<-0.0058 -18.3114 3.9355>
   rotation=<rot 0 89.9632 -89.9336>
  }]
 pelvis [{shapeType=Capsule, length=7.2189, radius=12.5014, center=<-3.3074 1.8094 0>, rotation=<rot 0 -90 0>}]
 root [{shapeType=Capsule, length=25.875, radius=49.9062, center=<-0.0001 52.2458 4.6268>, rotation=<rot 0 0 5.018>}]
 spine_01 [{shapeType=Capsule, length=6.9246, radius=9.8159, center=<0 2.2802 -0.0001>}]
 spine_02 [{shapeType=Capsule, length=4.4705, radius=9.755, center=<0 1.7189 -0.0001>}]
 spine_03 [{shapeType=Capsule, length=7.1914, radius=8.5382, center=<0 1.1599 -0.0001>}]
 spine_04 [{shapeType=Capsule, length=9.7013, radius=7.968, center=<0.6532 1.8714 0>}]
 spine_05 [{shapeType=Capsule, length=15.8663, radius=8.7969, rotation=<rot 0 -90 0>}]
 thigh_l [{shapeType=Capsule, length=26.5, radius=8.5, shapeName=thigh_r_capsule, center=<-20 -0.5256 0>, rotation=<rot 90 0 0>}]
 thigh_r [{shapeType=Capsule, length=26.5, radius=8.5, shapeName=thigh_l_capsule, center=<20 0.5256 0>, rotation=<rot 90 180 0>}]
 upperarm_l [{shapeType=Capsule, length=18.0812, radius=4.1347, center=<10.8056 -0.4936 0.3456>, rotation=<rot 90 0 0>}]
 upperarm_r [{shapeType=Capsule, length=18.0812, radius=4.1347, center=<-10.8056 0.4936 -0.3456>, rotation=<rot 90 0 0>}]
}
constraints [
 {
  bone1Name=spine_01
  bone2Name=pelvis
  swing1LimitDegrees=45
  swing1Motion=Free
  swing2LimitDegrees=45
  swing2Motion=Free
  twistLimitDegrees=45
  twistMotion=Free
 }
 {
  bone1Name=spine_02
  bone2Name=spine_01
  swing1LimitDegrees=45
  swing1Motion=Free
  swing2LimitDegrees=45
  swing2Motion=Free
  twistLimitDegrees=45
  twistMotion=Free
 }
 {
  bone1Name=spine_03
  bone2Name=spine_02
  swing1LimitDegrees=45
  swing1Motion=Free
  swing2LimitDegrees=45
  swing2Motion=Free
  twistLimitDegrees=45
  twistMotion=Free
 }
 {
  bone1Name=spine_04
  bone2Name=spine_03
  swing1LimitDegrees=45
  swing1Motion=Free
  swing2LimitDegrees=45
  swing2Motion=Free
  twistLimitDegrees=45
  twistMotion=Free
 }
 {
  bone1Name=spine_05
  bone2Name=spine_04
  swing1LimitDegrees=45
  swing1Motion=Free
  swing2LimitDegrees=45
  swing2Motion=Free
  twistLimitDegrees=45
  twistMotion=Free
 }
 {
  bone1Name=upperarm_l
  bone2Name=spine_05
  swing1LimitDegrees=45
  swing1Motion=Free
  swing2LimitDegrees=45
  swing2Motion=Free
  twistLimitDegrees=45
  twistMotion=Free
 }
 {
  bone1Name=lowerarm_l
  bone2Name=upperarm_l
  swing1LimitDegrees=45
  swing1Motion=Free
  swing2LimitDegrees=45
  swing2Motion=Free
  twistLimitDegrees=45
  twistMotion=Free
 }
 {
  bone1Name=upperarm_r
  bone2Name=spine_05
  swing1LimitDegrees=45
  swing1Motion=Free
  swing2LimitDegrees=45
  swing2Motion=Free
  twistLimitDegrees=45
  twistMotion=Free
 }
 {
  bone1Name=lowerarm_r
  bone2Name=upperarm_r
  swing1LimitDegrees=45
  swing1Motion=Free
  swing2LimitDegrees=45
  swing2Motion=Free
  twistLimitDegrees=45
  twistMotion=Free
 }
 {
  bone1Name=thigh_l
  bone2Name=pelvis
  swing1LimitDegrees=45
  swing1Motion=Free
  swing2LimitDegrees=45
  swing2Motion=Free
  twistLimitDegrees=45
  twistMotion=Free
 }
 {
  bone1Name=calf_l
  bone2Name=thigh_l
  swing1LimitDegrees=45
  swing1Motion=Free
  swing2LimitDegrees=45
  swing2Motion=Free
  twistLimitDegrees=45
  twistMotion=Free
 }
 {
  bone1Name=thigh_r
  bone2Name=pelvis
  swing1LimitDegrees=45
  swing1Motion=Free
  swing2LimitDegrees=45
  swing2Motion=Free
  twistLimitDegrees=45
  twistMotion=Free
 }
 {
  bone1Name=calf_r
  bone2Name=thigh_r
  swing1LimitDegrees=45
  swing1Motion=Free
  swing2LimitDegrees=45
  swing2Motion=Free
  twistLimitDegrees=45
  twistMotion=Free
 }
 {
  bone1Name=dyn_chain_01_r_MSTR
  bone2Name=pack
  swing1LimitDegrees=45
  swing1Motion=Free
  swing2LimitDegrees=45
  swing2Motion=Free
  twistLimitDegrees=45
  twistMotion=Free
 }
 {bone1Name=dyn_chain_01_r, bone2Name=dyn_chain_01_r_MSTR, swing1LimitDegrees=35, swing2LimitDegrees=35, twistLimitDegrees=5}
 {bone1Name=dyn_chain_02_r, bone2Name=dyn_chain_01_r, swing1LimitDegrees=35, swing2LimitDegrees=35, twistLimitDegrees=5}
 {bone1Name=dyn_chain_03_r, bone2Name=dyn_chain_02_r, swing1LimitDegrees=35, swing2LimitDegrees=35, twistLimitDegrees=5}
 {bone1Name=dyn_chain_04_r, bone2Name=dyn_chain_03_r, swing1LimitDegrees=35, swing2LimitDegrees=35, twistLimitDegrees=5}
 {bone1Name=dyn_chain_05_r, bone2Name=dyn_chain_04_r, swing1LimitDegrees=35, swing2LimitDegrees=35, twistLimitDegrees=5}
 {bone1Name=dyn_chain_06_r, bone2Name=dyn_chain_05_r, swing1LimitDegrees=35, swing2LimitDegrees=35, twistLimitDegrees=5}
 {bone1Name=dyn_chain_07_r, bone2Name=dyn_chain_06_r, swing1LimitDegrees=35, swing2LimitDegrees=35, twistLimitDegrees=5}
 {
  bone1Name=dyn_chain_01_l_MSTR
  bone2Name=pack
  swing1LimitDegrees=45
  swing1Motion=Free
  swing2LimitDegrees=45
  swing2Motion=Free
  twistLimitDegrees=45
  twistMotion=Free
 }
 {bone1Name=dyn_chain_01_l, bone2Name=dyn_chain_01_l_MSTR, swing1LimitDegrees=35, swing2LimitDegrees=35, twistLimitDegrees=5}
 {bone1Name=dyn_chain_02_l, bone2Name=dyn_chain_01_l, swing1LimitDegrees=35, swing2LimitDegrees=35, twistLimitDegrees=5}
 {bone1Name=dyn_chain_03_l, bone2Name=dyn_chain_02_l, swing1LimitDegrees=35, swing2LimitDegrees=35, twistLimitDegrees=5}
 {bone1Name=dyn_chain_04_l, bone2Name=dyn_chain_03_l, swing1LimitDegrees=35, swing2LimitDegrees=35, twistLimitDegrees=5}
 {bone1Name=dyn_chain_05_l, bone2Name=dyn_chain_04_l, swing1LimitDegrees=35, swing2LimitDegrees=35, twistLimitDegrees=5}
 {bone1Name=dyn_chain_06_l, bone2Name=dyn_chain_05_l, swing1LimitDegrees=35, swing2LimitDegrees=35, twistLimitDegrees=5}
 {bone1Name=dyn_chain_07_l, bone2Name=dyn_chain_06_l, swing1LimitDegrees=35, swing2LimitDegrees=35, twistLimitDegrees=5}
 {
  bone1Name=pelvis
  bone2Name=root
  swing1LimitDegrees=45
  swing1Motion=Free
  swing2LimitDegrees=45
  swing2Motion=Free
  twistLimitDegrees=45
  twistMotion=Free
 }
 {
  bone1Name=pack
  bone2Name=root
  swing1LimitDegrees=45
  swing1Motion=Free
  swing2LimitDegrees=45
  swing2Motion=Free
  twistLimitDegrees=45
  twistMotion=Free
 }
 {
  bone1Name=foot_r
  bone2Name=calf_r
  swing1LimitDegrees=45
  swing1Motion=Free
  swing2LimitDegrees=45
  swing2Motion=Free
  twistLimitDegrees=45
  twistMotion=Free
 }
]
