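# Animation Blueprint for the SunBeam backpack cosmetic, driving the secondary motion of its dangling chains.
# Each update copies the parent mesh's body bone transforms, sets gravity, and generates random chain wobble while skydiving.
# The output path applies those bones, a spine-weighted root constraint, movement-gated chain Look Ats and wobble, then RigidBody simulation.
# Diamond AnimDynamics, state-based pose blends and per-state physics tuning are present but disconnected.
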AnimBlueprint /BRCosmetics/Accessories/FORT_Backpacks/Backpack_F_MED_SunBeam/Meshes/F_MED_SunBeam_Pack_AnimBP
target_skeleton=&/F_MED_SunBeam_Pack_Skeleton
graphs [
 AnimGraph {nodes [
   n1 = "Transform (Modify) Bone - Bone: pack"(ComponentPose=Local To Component(LocalPose=Use cached pose 'MainPose'()), Translation="0,-2,1", Rotation="0,0,17") {
    rotationMode=BMM_Additive
    scaleMode=BMM_Additive
    translationMode=BMM_Additive
    alphaBoolBlend {blendOption=Linear}
    boneToModify {boneName=pack}
    rotation {roll=17}
    translation {y=-2, z=1}
   }
   n2 = "Transform (Modify) Bone - Bone: dyn_chain_collider_01"(ComponentPose=@n1, Alpha=1, Translation="0,-4,-30", Rotation="0,0,0", Scale="1,1,1") {
    rotationMode=BMM_Additive
    scaleMode=BMM_Additive
    translationMode=BMM_Additive
    alphaBoolBlend {blendOption=Linear}
    boneToModify {boneName=dyn_chain_collider_01}
    translation {y=-4, z=-30}
   }
   Blend Poses by bool(BlendPose_0=Component To Local(ComponentPose=@n2), BlendTime_0=1, BlendTime_1=1, bActiveValue=Get bIsTacticalSprinting()) {blendType=Linear}
   n3 = Local To Component(LocalPose=Use cached pose 'MainPose'())
   n4 = NOT Boolean(A=Get bIsPlayingMeleeAnim())
   n5 = Get bIsSprinting()
   n6 = "Transform (Modify) Bone - Bone: dyn_chain_collider_01"(ComponentPose=Local To Component(LocalPose=Use cached pose 'MainPose'()), Alpha=1, Translation="1,4,-15", Rotation="2,0,20", Scale="1,1,1") {
    rotationMode=BMM_Additive
    scaleMode=BMM_Additive
    translationMode=BMM_Additive
    alphaBoolBlend {blendOption=Linear}
    boneToModify {boneName=dyn_chain_collider_01}
    rotation {pitch=2, roll=20}
    translation=<1 4 -15>
   }
   n7 = OR Boolean!(A=Get bIsCrouching(), B=Get bIsSprinting(), C:bool=Get bIsPlayingMeleeAnim()) -> (ReturnValue)
   n8 = Blend Poses by bool(
    BlendPose_0=Component To Local(ComponentPose=@n6)
    BlendPose_1=Blend Poses by bool(
     BlendPose_0=Component To Local()
     BlendPose_1=Component To Local(ComponentPose=@n3)
     BlendTime_0=0.3
     BlendTime_1=0.3
     bActiveValue=AND Boolean(A=@n5, B=@n4)
    ) {blendType=Linear}
    BlendTime_0=0.3
    BlendTime_1=0.3
    bActiveValue=AND Boolean(A=Get bIsMoving2D(), B=NOT Boolean(A=@n7))
   ) {blendType=Linear}
   n9 = "Transform (Modify) Bone - Bone: pack"(ComponentPose=Local To Component(LocalPose=Use cached pose 'MainPose'()), Translation="0,-2,1", Rotation="0,0,10") {
    rotationMode=BMM_Additive
    scaleMode=BMM_Additive
    translationMode=BMM_Additive
    alphaBoolBlend {blendOption=Linear}
    boneToModify {boneName=pack}
    rotation {roll=10}
    translation {y=-2, z=1}
   }
   n10 = "Transform (Modify) Bone - Bone: dyn_chain_collider_01"(ComponentPose=@n9, Alpha=1, Translation="15,-8,-52", Rotation="0,0,35", Scale="1,1,1") {
    rotationMode=BMM_Additive
    scaleMode=BMM_Additive
    translationMode=BMM_Additive
    alphaBoolBlend {blendOption=Linear}
    boneToModify {boneName=dyn_chain_collider_01}
    rotation {roll=35}
    translation=<15 -8 -52>
   }
   n11 = Component To Local(ComponentPose=@n10)
   n12 = Get bIsPlayingMeleeAnim()
   n13 = "Transform (Modify) Bone - Bone: pack"(ComponentPose=Local To Component(LocalPose=Use cached pose 'MainPose'()), Translation="0,-4,1", Rotation="0,0,25") {
    rotationMode=BMM_Additive
    scaleMode=BMM_Additive
    translationMode=BMM_Additive
    alphaBoolBlend {blendOption=Linear}
    boneToModify {boneName=pack}
    rotation {roll=25}
    translation {y=-4, z=1}
   }
   n14 = "Transform (Modify) Bone - Bone: dyn_chain_collider_01"(ComponentPose=@n13, Alpha=1, Translation="0,-20,-63", Rotation="0,0,25", Scale="1,1,1") {
    rotationMode=BMM_Additive
    scaleMode=BMM_Additive
    translationMode=BMM_Additive
    alphaBoolBlend {blendOption=Linear}
    boneToModify {boneName=dyn_chain_collider_01}
    rotation {roll=25}
    translation {y=-20, z=-63}
   }
   n15 = Get bIsSkyDiving()
   n16 = Get bIsSkydiveDiveMode()
   n17 = Get bIsSkydiveFloating()
   Local To Component(LocalPose=Blend Poses by bool(
     BlendPose_0=Component To Local(ComponentPose=@n14)
     BlendPose_1=Blend Poses by bool(BlendPose_0=@n11, BlendPose_1=@n8, BlendTime_0=0.3, BlendTime_1=0.3, bActiveValue=@n12) {blendType=Linear}
     BlendTime_0=0.5
     BlendTime_1=0.5
     bActiveValue=OR Boolean(A=@n15, B=@n17, C=@n16)
    ) {blendType=Linear})
   n18 = "Transform (Modify) Bone - Bone: dyn_chain_collider_01"(ComponentPose=Local To Component(LocalPose=Use cached pose 'MainPose'()), Alpha=1, Translation="8,4,-15", Rotation="2,0,7", Scale="1,1,1") {
    rotationMode=BMM_Additive
    scaleMode=BMM_Additive
    translationMode=BMM_Additive
    alphaBoolBlend {blendOption=Linear}
    boneToModify {boneName=dyn_chain_collider_01}
    rotation {pitch=2, roll=7}
    translation=<8 4 -15>
   }
   "Transform (Modify) Bone - Bone: dyn_chain_collider_01"(ComponentPose=@n18, Alpha=1, Translation="0,78,10", Rotation="0,0,6", Scale="1,1,1") {
    rotationMode=BMM_Additive
    scaleMode=BMM_Additive
    translationMode=BMM_Additive
    alphaBoolBlend {blendOption=Linear}
    boneToModify {boneName=dyn_chain_collider_01}
    rotation {roll=6}
    translation {y=78, z=10}
   }
   n19 = Get bIsMoving2D()
   n20 = Look At - Bone: dyn_chain_01_r_MSTR!(ComponentPose=Local To Component(LocalPose=Use cached pose 'anim cache'()), bAlphaBoolEnabled:bool=@n19) -> (Pose) {
    alphaInputType=Bool
    alphaBoolBlend {blendInTime=0.2, blendOption=Linear, blendOutTime=0.3}
    boneToModify {boneName=dyn_chain_01_r_MSTR}
    lookAtLocation=<0 -60>
    lookAtTarget {boneReference {boneName=pelvis}}
    lookAt_Axis {axis=<1 0>}
   }
   n21 = Look At - Bone: dyn_chain_01_l_MSTR!(ComponentPose=@n20, bAlphaBoolEnabled:bool=@n19) -> (Pose) {
    alphaInputType=Bool
    alphaBoolBlend {blendInTime=0.2, blendOption=Linear, blendOutTime=0.3}
    boneToModify {boneName=dyn_chain_01_l_MSTR}
    lookAtLocation=<0 -60>
    lookAtTarget {boneReference {boneName=pelvis}}
    lookAt_Axis {axis=<1 0>}
   }
   n22 = Get bIsSkyDiving()
   n23 = "Transform (Modify) Bone - Bone: dyn_chain_01_l_MSTR"!(ComponentPose=@n21, bAlphaBoolEnabled:bool=@n22, Translation="0,0,0", Rotation=Get Rotator_01() -> (Rotator_01:Rotator), Scale="1,1,1") -> (Pose) {
    alphaInputType=Bool
    rotationMode=BMM_Additive
    rotationSpace=BCS_BoneSpace
    alphaBoolBlend {blendInTime=0.1, blendOption=Linear, blendOutTime=0.1}
    boneToModify {boneName=dyn_chain_01_l_MSTR}
   }
   n24 = "Transform (Modify) Bone - Bone: dyn_chain_01_l_MSTR"(ComponentPose=@n23, bAlphaBoolEnabled=@n22, Translation="0,0,0", Rotation=Get Rotator_02() -> (Rotator_02:Rotator), Scale="1,1,1") {
    alphaInputType=Bool
    rotationMode=BMM_Additive
    rotationSpace=BCS_BoneSpace
    alphaBoolBlend {blendInTime=0.1, blendOption=Linear, blendOutTime=0.1}
    boneToModify {boneName=dyn_chain_01_l_MSTR}
   }
   n25 = RigidBody!(
    ComponentPose=@n24
    Alpha
    OverrideWorldGravity:Vector=Get GravityOverride()
    ExternalForce
    ComponentLinearAccScale:Vector="0.3,0.3,0.5"
    ComponentLinearVelScale:Vector="0.3,0.3,0.5"
   ) -> (Pose) {
    bClampLinearTranslationLimitToRefPose=true
    bOverrideWorldGravity=true
    overlapChannel=ECC_Pawn
    overridePhysicsAsset=&/F_MED_SunBeam_Pack_Female_Physics
    simulationSpace=BaseBoneSpace
    alphaBoolBlend {blendOption=Linear}
    baseBoneRef {boneName=root}
    componentAppliedLinearAccClamp=<6000 6000 6000>
    componentLinearAccScale=<0.6 0.5 0.5>
    componentLinearVelScale=<0.6 0.5 0.5>
    overrideWorldGravity {z=-980}
   }
   Output Pose(Result=Component To Local(ComponentPose=@n25)) {layerGroup=None}
   n26 = "Transform (Modify) Bone - Bone: body_root"(
    ComponentPose=Local To Component(LocalPose=Input Pose())
    Alpha=1
    Translation=Get bodyRoot_tran() -> (bodyRoot_tran:Vector)
    Rotation=Get bodyRoot_rot() -> (bodyRoot_rot:Rotator)
    Scale="1,1,1"
   ) {
    rotationMode=BMM_Replace
    rotationSpace=BCS_WorldSpace
    scaleSpace=BCS_BoneSpace
    translationMode=BMM_Replace
    translationSpace=BCS_WorldSpace
    alphaBoolBlend {blendOption=Linear}
    alphaScaleBiasClamp {inRange {max=0, min=-40}, outRange {max=0, min=1}}
    boneToModify {boneName=body_root}
   }
   n27 = "Transform (Modify) Bone - Bone: pelvis"(
    ComponentPose=@n26
    Alpha=1
    Translation=Get pelvis_tran() -> (pelvis_tran:Vector)
    Rotation=Get pelvis_rot() -> (pelvis_rot:Rotator)
    Scale="1,1,1"
   ) {
    rotationMode=BMM_Replace
    rotationSpace=BCS_WorldSpace
    scaleSpace=BCS_BoneSpace
    translationMode=BMM_Replace
    translationSpace=BCS_WorldSpace
    alphaBoolBlend {blendOption=Linear}
    alphaScaleBiasClamp {inRange {max=0, min=-40}, outRange {max=0, min=1}}
    boneToModify {boneName=pelvis}
   }
   n28 = "Transform (Modify) Bone - Bone: floorCOL"(
    ComponentPose=@n27
    Alpha=1
    Translation=Get root_trans() -> (root_trans:Vector)
    Rotation=Get root_rot() -> (root_rot:Rotator)
    Scale="1,1,1"
   ) {
    rotationMode=BMM_Replace
    rotationSpace=BCS_WorldSpace
    scaleSpace=BCS_BoneSpace
    translationMode=BMM_Replace
    translationSpace=BCS_WorldSpace
    alphaBoolBlend {blendOption=Linear}
    alphaScaleBiasClamp {inRange {max=0, min=-40}, outRange {max=0, min=1}}
    boneToModify {boneName=floorCOL}
   }
   n29 = "Transform (Modify) Bone - Bone: thigh_l"(
    ComponentPose=@n28
    Alpha=1
    Translation=Get thigh_l_tran() -> (thigh_l_tran:Vector)
    Rotation=Get thigh_l_rot() -> (thigh_l_rot:Rotator)
    Scale="1,1,1"
   ) {
    rotationMode=BMM_Replace
    rotationSpace=BCS_WorldSpace
    scaleSpace=BCS_BoneSpace
    translationMode=BMM_Replace
    translationSpace=BCS_WorldSpace
    alphaBoolBlend {blendOption=Linear}
    alphaScaleBiasClamp {inRange {max=0, min=-40}, outRange {max=0, min=1}}
    boneToModify {boneName=thigh_l}
   }
   n30 = "Transform (Modify) Bone - Bone: thigh_r"(
    ComponentPose=@n29
    Alpha=1
    Translation=Get thigh_r_tran() -> (thigh_r_tran:Vector)
    Rotation=Get thigh_r_rot() -> (thigh_r_rot:Rotator)
    Scale="1,1,1"
   ) {
    rotationMode=BMM_Replace
    rotationSpace=BCS_WorldSpace
    scaleSpace=BCS_BoneSpace
    translationMode=BMM_Replace
    translationSpace=BCS_WorldSpace
    alphaBoolBlend {blendOption=Linear}
    alphaScaleBiasClamp {inRange {max=0, min=-40}, outRange {max=0, min=1}}
    boneToModify {boneName=thigh_r}
   }
   n31 = "Transform (Modify) Bone - Bone: calf_l"(
    ComponentPose=@n30
    Alpha=1
    Translation=Get calf_l_tran() -> (calf_l_tran:Vector)
    Rotation=Get calf_l_rot() -> (calf_l_rot:Rotator)
    Scale="1,1,1"
   ) {
    rotationMode=BMM_Replace
    rotationSpace=BCS_WorldSpace
    scaleSpace=BCS_BoneSpace
    translationMode=BMM_Replace
    translationSpace=BCS_WorldSpace
    alphaBoolBlend {blendOption=Linear}
    alphaScaleBiasClamp {inRange {max=0, min=-40}, outRange {max=0, min=1}}
    boneToModify {boneName=calf_l}
   }
   n32 = "Transform (Modify) Bone - Bone: calf_r"(
    ComponentPose=@n31
    Alpha=1
    Translation=Get calf_r_tran() -> (calf_r_tran:Vector)
    Rotation=Get calf_r_rot() -> (calf_r_rot:Rotator)
    Scale="1,1,1"
   ) {
    rotationMode=BMM_Replace
    rotationSpace=BCS_WorldSpace
    scaleSpace=BCS_BoneSpace
    translationMode=BMM_Replace
    translationSpace=BCS_WorldSpace
    alphaBoolBlend {blendOption=Linear}
    alphaScaleBiasClamp {inRange {max=0, min=-40}, outRange {max=0, min=1}}
    boneToModify {boneName=calf_r}
   }
   n33 = "Transform (Modify) Bone - Bone: spine_01"(
    ComponentPose=@n32
    Alpha=1
    Translation=Get spine_01_tran() -> (spine_01_tran:Vector)
    Rotation=Get spine_01_rot() -> (spine_01_rot:Rotator)
    Scale="1,1,1"
   ) {
    rotationMode=BMM_Replace
    rotationSpace=BCS_WorldSpace
    scaleSpace=BCS_BoneSpace
    translationMode=BMM_Replace
    translationSpace=BCS_WorldSpace
    alphaBoolBlend {blendOption=Linear}
    alphaScaleBiasClamp {inRange {max=0, min=-40}, outRange {max=0, min=1}}
    boneToModify {boneName=spine_01}
   }
   n34 = "Transform (Modify) Bone - Bone: spine_02"(
    ComponentPose=@n33
    Alpha=1
    Translation=Get spine_02_tran() -> (spine_02_tran:Vector)
    Rotation=Get spine_02_rot() -> (spine_02_rot:Rotator)
    Scale="1,1,1"
   ) {
    rotationMode=BMM_Replace
    rotationSpace=BCS_WorldSpace
    scaleSpace=BCS_BoneSpace
    translationMode=BMM_Replace
    translationSpace=BCS_WorldSpace
    alphaBoolBlend {blendOption=Linear}
    alphaScaleBiasClamp {inRange {max=0, min=-40}, outRange {max=0, min=1}}
    boneToModify {boneName=spine_02}
   }
   n35 = "Transform (Modify) Bone - Bone: spine_03"(
    ComponentPose=@n34
    Alpha=1
    Translation=Get spine_03_tran() -> (spine_03_tran:Vector)
    Rotation=Get spine_03_rot() -> (spine_03_rot:Rotator)
    Scale="1,1,1"
   ) {
    rotationMode=BMM_Replace
    rotationSpace=BCS_WorldSpace
    scaleSpace=BCS_BoneSpace
    translationMode=BMM_Replace
    translationSpace=BCS_WorldSpace
    alphaBoolBlend {blendOption=Linear}
    alphaScaleBiasClamp {inRange {max=0, min=-40}, outRange {max=0, min=1}}
    boneToModify {boneName=spine_03}
   }
   n36 = "Transform (Modify) Bone - Bone: spine_04"(
    ComponentPose=@n35
    Alpha=1
    Translation=Get spine_04_tran() -> (spine_04_tran:Vector)
    Rotation=Get spine_04_rot() -> (spine_04_rot:Rotator)
    Scale="1,1,1"
   ) {
    rotationMode=BMM_Replace
    rotationSpace=BCS_WorldSpace
    scaleSpace=BCS_BoneSpace
    translationMode=BMM_Replace
    translationSpace=BCS_WorldSpace
    alphaBoolBlend {blendOption=Linear}
    alphaScaleBiasClamp {inRange {max=0, min=-40}, outRange {max=0, min=1}}
    boneToModify {boneName=spine_04}
   }
   n37 = "Transform (Modify) Bone - Bone: spine_05"(
    ComponentPose=@n36
    Alpha=1
    Translation=Get spine_05_tran() -> (spine_05_tran:Vector)
    Rotation=Get spine_05_rot() -> (spine_05_rot:Rotator)
    Scale="1,1,1"
   ) {
    rotationMode=BMM_Replace
    rotationSpace=BCS_WorldSpace
    scaleSpace=BCS_BoneSpace
    translationMode=BMM_Replace
    translationSpace=BCS_WorldSpace
    alphaBoolBlend {blendOption=Linear}
    alphaScaleBiasClamp {inRange {max=0, min=-40}, outRange {max=0, min=1}}
    boneToModify {boneName=spine_05}
   }
   n38 = "Transform (Modify) Bone - Bone: clavicle_r"(
    ComponentPose=@n37
    Alpha=1
    Translation=Get clavicle_r_tran() -> (clavicle_r_tran:Vector)
    Rotation=Get clavicel_r_rot() -> (clavicel_r_rot:Rotator)
    Scale="1,1,1"
   ) {
    rotationMode=BMM_Replace
    rotationSpace=BCS_WorldSpace
    scaleSpace=BCS_BoneSpace
    translationMode=BMM_Replace
    translationSpace=BCS_WorldSpace
    alphaBoolBlend {blendOption=Linear}
    alphaScaleBiasClamp {inRange {max=0, min=-40}, outRange {max=0, min=1}}
    boneToModify {boneName=clavicle_r}
   }
   n39 = "Transform (Modify) Bone - Bone: clavicle_l"(
    ComponentPose=@n38
    Alpha=1
    Translation=Get clavicle_l_tran() -> (clavicle_l_tran:Vector)
    Rotation=Get clavicle_l_rot() -> (clavicle_l_rot:Rotator)
    Scale="1,1,1"
   ) {
    rotationMode=BMM_Replace
    rotationSpace=BCS_WorldSpace
    scaleSpace=BCS_BoneSpace
    translationMode=BMM_Replace
    translationSpace=BCS_WorldSpace
    alphaBoolBlend {blendOption=Linear}
    alphaScaleBiasClamp {inRange {max=0, min=-40}, outRange {max=0, min=1}}
    boneToModify {boneName=clavicle_l}
   }
   n40 = "Transform (Modify) Bone - Bone: upperarm_r"(
    ComponentPose=@n39
    Alpha=1
    Translation=Get upperarm_r_tran() -> (upperarm_r_tran:Vector)
    Rotation=Get upperarm_r_rot() -> (upperarm_r_rot:Rotator)
    Scale="1,1,1"
   ) {
    rotationMode=BMM_Replace
    rotationSpace=BCS_WorldSpace
    scaleSpace=BCS_BoneSpace
    translationMode=BMM_Replace
    translationSpace=BCS_WorldSpace
    alphaBoolBlend {blendOption=Linear}
    alphaScaleBiasClamp {inRange {max=0, min=-40}, outRange {max=0, min=1}}
    boneToModify {boneName=upperarm_r}
   }
   n41 = "Transform (Modify) Bone - Bone: upperarm_l"(
    ComponentPose=@n40
    Alpha=1
    Translation=Get upperarm_l_tran() -> (upperarm_l_tran:Vector)
    Rotation=Get upperarm_l_rot() -> (upperarm_l_rot:Rotator)
    Scale="1,1,1"
   ) {
    rotationMode=BMM_Replace
    rotationSpace=BCS_WorldSpace
    scaleSpace=BCS_BoneSpace
    translationMode=BMM_Replace
    translationSpace=BCS_WorldSpace
    alphaBoolBlend {blendOption=Linear}
    alphaScaleBiasClamp {inRange {max=0, min=-40}, outRange {max=0, min=1}}
    boneToModify {boneName=upperarm_l}
   }
   n42 = "Transform (Modify) Bone - Bone: lowerarm_l"(
    ComponentPose=@n41
    Alpha=1
    Translation=Get lowerarm_l_tran() -> (lowerarm_l_tran:Vector)
    Rotation=Get lowerarm_l_rot() -> (lowerarm_l_rot:Rotator)
    Scale="1,1,1"
   ) {
    rotationMode=BMM_Replace
    rotationSpace=BCS_WorldSpace
    scaleSpace=BCS_BoneSpace
    translationMode=BMM_Replace
    translationSpace=BCS_WorldSpace
    alphaBoolBlend {blendOption=Linear}
    alphaScaleBiasClamp {inRange {max=0, min=-40}, outRange {max=0, min=1}}
    boneToModify {boneName=lowerarm_l}
   }
   n43 = "Transform (Modify) Bone - Bone: lowerarm_r"(
    ComponentPose=@n42
    Alpha=1
    Translation=Get lowerarm_r_tran() -> (lowerarm_r_tran:Vector)
    Rotation=Get lowerarm_r_rot() -> (lowerarm_r_rot:Rotator)
    Scale="1,1,1"
   ) {
    rotationMode=BMM_Replace
    rotationSpace=BCS_WorldSpace
    scaleSpace=BCS_BoneSpace
    translationMode=BMM_Replace
    translationSpace=BCS_WorldSpace
    alphaBoolBlend {blendOption=Linear}
    alphaScaleBiasClamp {inRange {max=0, min=-40}, outRange {max=0, min=1}}
    boneToModify {boneName=lowerarm_r}
   }
   n44 = "Constraint - root (RRR)"!(ComponentPose=@n43, Alpha=1, ConstraintWeights_0=0.4, ConstraintWeights_1:float=0.5, ConstraintWeights_2:float=0.1) -> (Pose) {
    alphaBoolBlend {blendOption=Linear}
    boneToModify {boneName=root}
    constraintSetup [
     {transformType=Rotation, targetBone {boneName=spine_05}}
     {transformType=Rotation, targetBone {boneName=spine_04}}
     {transformType=Rotation, targetBone {boneName=spine_03}}
    ]
    constraintWeights [1,1,1]
   }
   Save cached pose 'anim cache'(Pose=Component To Local(ComponentPose=@n44)) {cacheName=anim cache}
   n45 = Anim Dynamics - Bone: dyn_big_diamond_01_m(ComponentPose=Local To Component(), Alpha=1, ExternalForce="0,0,0") {
    angularBiasOverride=1
    angularDampingOverride=0.9
    angularSpringConstant=2
    bOverrideAngularDamping=true
    bUseGravityOverride=true
    simulationSpace=RootRelative
    alphaBoolBlend {blendOption=Linear}
    boundBone {boneName=dyn_big_diamond_01_m}
    componentLinearAccScale=<0.3 0.3 0.3>
    componentLinearVelScale=<0.3 0.3 0.3>
    gravityOverride {z=-800}
    physicsBodyDefinitions [{
      boundBone {boneName=dyn_big_diamond_01_m}
      boxExtents=<5 10 10>
      constraintSetup {angularLimitsMax=<10 15 9>, angularLimitsMin=<-10 -15 -55>, angularTarget=<1 0 0>}
      localJointOffset=<8 0 0>
     }]
   }
   n46 = Anim Dynamics - Bone: dyn_big_diamond_01_r(ComponentPose=@n45, Alpha=1, ExternalForce="0,0,0") {
    angularBiasOverride=1
    angularDampingOverride=0.9
    angularSpringConstant=2
    bOverrideAngularDamping=true
    bUseGravityOverride=true
    simulationSpace=RootRelative
    alphaBoolBlend {blendOption=Linear}
    boundBone {boneName=dyn_big_diamond_01_r}
    componentLinearAccScale=<0.35 0.35 0.35>
    componentLinearVelScale=<0.35 0.35 0.35>
    gravityOverride {z=-1000}
    physicsBodyDefinitions [{
      boundBone {boneName=dyn_big_diamond_01_r}
      boxExtents=<4 9 9>
      constraintSetup {angularLimitsMax=<11 14 8>, angularLimitsMin=<-8 -13 -55>, angularTarget=<1 0 0>}
      localJointOffset=<8 0 0>
     }]
   }
   n47 = Anim Dynamics - Bone: dyn_big_diamond_01_l(ComponentPose=@n46, Alpha=1, ExternalForce="0,0,0") {
    angularBiasOverride=1
    angularDampingOverride=0.9
    angularSpringConstant=2
    bOverrideAngularDamping=true
    bUseGravityOverride=true
    simulationSpace=RootRelative
    alphaBoolBlend {blendOption=Linear}
    boundBone {boneName=dyn_big_diamond_01_l}
    componentLinearAccScale=<0.25 0.25 0.25>
    componentLinearVelScale=<0.25 0.25 0.25>
    gravityOverride {z=-850}
    physicsBodyDefinitions [{
      boundBone {boneName=dyn_big_diamond_01_l}
      boxExtents=<4.5 10.5 10.5>
      constraintSetup {angularLimitsMax=<11 14 9>, angularLimitsMin=<-9 -14 -55>, angularTarget=<1 0 0>}
      localJointOffset=<8 0 0>
     }]
   }
   n48 = Anim Dynamics - Bone: dyn_small_diamond_01_l(ComponentPose=@n47, Alpha=1, ExternalForce="0,0,0") {
    angularBiasOverride=1
    angularSpringConstant=1
    bAngularSpring=true
    bUseGravityOverride=true
    simulationSpace=RootRelative
    alphaBoolBlend {blendOption=Linear}
    boundBone {boneName=dyn_small_diamond_01_l}
    componentLinearAccScale=<0.25 0.25 0.25>
    componentLinearVelScale=<0.25 0.25 0.25>
    gravityOverride {z=-450}
    physicsBodyDefinitions [{
      boundBone {boneName=dyn_small_diamond_01_l}
      boxExtents=<2 5 5>
      constraintSetup {angularLimitsMax=<15 25 25>, angularLimitsMin=<-15 -25 -100>, angularTarget=<1 0 0>}
      localJointOffset=<1 0 0>
     }]
   }
   n49 = Anim Dynamics - Bone: dyn_small_diamond_01_r(ComponentPose=@n48, Alpha=1, ExternalForce="0,0,0") {
    angularBiasOverride=1
    angularSpringConstant=2
    bUseGravityOverride=true
    simulationSpace=RootRelative
    alphaBoolBlend {blendOption=Linear}
    boundBone {boneName=dyn_small_diamond_01_r}
    componentLinearAccScale=<0.3 0.3 0.3>
    componentLinearVelScale=<0.3 0.3 0.3>
    gravityOverride {z=-980}
    physicsBodyDefinitions [{
      boundBone {boneName=dyn_small_diamond_01_r}
      boxExtents=<3 6 6>
      constraintSetup {angularLimitsMax=<20 30 35>, angularLimitsMin=<-20 -30 -110>, angularTarget=<1 0 0>}
      localJointOffset=<2 0 0>
     }]
   }
   n50 = Copy Bone - Source Bone: dyn_small_diamond_01_l - Target Bone: dyn_small_diamond_02_l!(
    ComponentPose:Component Space Pose Link Structure=@n49
    Alpha:float=0.7
    bCopyTranslation:bool=False
    bCopyRotation:bool=true
    bCopyScale:bool=False
   ) -> (Pose:Component Space Pose Link Structure) {
    is=title AnimGraphNode_CopyBone
    alphaBoolBlend {blendOption=Linear}
    sourceBone {boneName=dyn_small_diamond_01_l}
    targetBone {boneName=dyn_small_diamond_02_l}
   }
   n51 = Copy Bone - Source Bone: dyn_small_diamond_01_r - Target Bone: dyn_small_diamond_03_l!(
    ComponentPose:Component Space Pose Link Structure=@n50
    Alpha:float=1.3
    bCopyTranslation:bool=False
    bCopyRotation:bool=true
    bCopyScale:bool=False
   ) -> (Pose:Component Space Pose Link Structure) {
    is=title AnimGraphNode_CopyBone
    alphaBoolBlend {blendOption=Linear}
    sourceBone {boneName=dyn_small_diamond_01_r}
    targetBone {boneName=dyn_small_diamond_03_l}
   }
   n52 = Copy Bone - Source Bone: dyn_small_diamond_01_l - Target Bone: dyn_small_diamond_04_l!(
    ComponentPose:Component Space Pose Link Structure=@n51
    Alpha:float=0.9
    bCopyTranslation:bool=False
    bCopyRotation:bool=true
    bCopyScale:bool=False
   ) -> (Pose:Component Space Pose Link Structure) {
    is=title AnimGraphNode_CopyBone
    alphaBoolBlend {blendOption=Linear}
    sourceBone {boneName=dyn_small_diamond_01_l}
    targetBone {boneName=dyn_small_diamond_04_l}
   }
   n53 = Copy Bone - Source Bone: dyn_small_diamond_01_r - Target Bone: dyn_small_diamond_05_l!(
    ComponentPose:Component Space Pose Link Structure=@n52
    Alpha:float=0.6
    bCopyTranslation:bool=False
    bCopyRotation:bool=true
    bCopyScale:bool=False
   ) -> (Pose:Component Space Pose Link Structure) {
    is=title AnimGraphNode_CopyBone
    alphaBoolBlend {blendOption=Linear}
    sourceBone {boneName=dyn_small_diamond_01_r}
    targetBone {boneName=dyn_small_diamond_05_l}
   }
   n54 = Copy Bone - Source Bone: dyn_small_diamond_01_l - Target Bone: dyn_small_diamond_02_r!(
    ComponentPose:Component Space Pose Link Structure=@n53
    Alpha:float=0.9
    bCopyTranslation:bool=False
    bCopyRotation:bool=true
    bCopyScale:bool=False
   ) -> (Pose:Component Space Pose Link Structure) {
    is=title AnimGraphNode_CopyBone
    alphaBoolBlend {blendOption=Linear}
    sourceBone {boneName=dyn_small_diamond_01_l}
    targetBone {boneName=dyn_small_diamond_02_r}
   }
   n55 = Copy Bone - Source Bone: dyn_small_diamond_01_r - Target Bone: dyn_small_diamond_03_r!(
    ComponentPose:Component Space Pose Link Structure=@n54
    Alpha:float=1.4
    bCopyTranslation:bool=False
    bCopyRotation:bool=true
    bCopyScale:bool=False
   ) -> (Pose:Component Space Pose Link Structure) {
    is=title AnimGraphNode_CopyBone
    alphaBoolBlend {blendOption=Linear}
    sourceBone {boneName=dyn_small_diamond_01_r}
    targetBone {boneName=dyn_small_diamond_03_r}
   }
   n56 = Copy Bone - Source Bone: dyn_small_diamond_01_l - Target Bone: dyn_small_diamond_04_r!(
    ComponentPose:Component Space Pose Link Structure=@n55
    Alpha:float=0.75
    bCopyTranslation:bool=False
    bCopyRotation:bool=true
    bCopyScale:bool=False
   ) -> (Pose:Component Space Pose Link Structure) {
    is=title AnimGraphNode_CopyBone
    alphaBoolBlend {blendOption=Linear}
    sourceBone {boneName=dyn_small_diamond_01_l}
    targetBone {boneName=dyn_small_diamond_04_r}
   }
   n57 = Copy Bone - Source Bone: dyn_small_diamond_01_r - Target Bone: dyn_small_diamond_03_r1!(
    ComponentPose:Component Space Pose Link Structure=@n56
    Alpha:float=1.5
    bCopyTranslation:bool=False
    bCopyRotation:bool=true
    bCopyScale:bool=False
   ) -> (Pose:Component Space Pose Link Structure) {
    is=title AnimGraphNode_CopyBone
    alphaBoolBlend {blendOption=Linear}
    sourceBone {boneName=dyn_small_diamond_01_r}
    targetBone {boneName=dyn_small_diamond_03_r1}
   }
   Save cached pose 'MainPose'(Pose=Component To Local(ComponentPose=@n57)) {cacheName=MainPose}
   Get Linear_Acc_Scale() -> (Linear_Acc_Scale:Vector)
   Get Linear_Vel_Scale() -> (Linear_Vel_Scale:Vector)
   Get External_Force() -> (External_Force:Vector)
  ]}
 EventGraph {nodes [
   Event BlueprintUpdateAnimation()
   > n1 = Sequence!(execute) -> (then_0, then_1, then_2:Exec, then_3:Exec)
   > n2 = Gravity Override State Function DEPRECATED(CharacterPartAnimInstance=Self-Reference(), GameTimeInSeconds=GetGameTimeInSeconds(), GravityOverrideParams=Get Gravity Override Params())
   > Set GravityOverride(GravityOverride=@n2)
   > n3 = Cast To FortPlayerPawn(Object=TryGetPawnOwner())
   n4 = GetGender!(self:Fort Player Pawn Object Reference=@n3) -> (IsMale:bool, IsFemale:bool)
   > Set IsMale(IsMale:bool=@n4) -> (Output_Get:bool)
   > Set IsFemale(IsFemale:bool=@n4.IsFemale) -> (Output_Get:bool)
   n5 = Branch(execute=@n1.then_2, Condition=Get bIsSkyDiving())
   n6 = GetGameTimeInSeconds()
   > n7 = RandomRotatorFunction!(
    execute
    GameTimeInSeconds:double=@n6
    PerlinScale:double=1
    Frequency:Rotator="5,3,0"
    Amplitude:Rotator="20,20,0"
    Offset:Rotator="0, 0, 0"
    TimeOffset:double=0
   ) -> (then, RandomRotator:Rotator)
   > Set Rotator_01(Rotator_01:Rotator=@n7) -> (Output_Get:Rotator)
   > n8 = RandomRotatorFunction(GameTimeInSeconds=@n6, PerlinScale=1, Frequency="3,1.5, 0", Amplitude="15,15, 0", Offset="0, 0, 0", TimeOffset=0)
   > Set Rotator_02(Rotator_02:Rotator=@n8) -> (Output_Get:Rotator)
   Set Rotator_01(execute=@n5.else, Rotator_01="0, 0, 0")
   > Set Rotator_02(Rotator_02="0, 0, 0")
   UpdateBodyJoinTransforms(execute=@n1.then_1)
   Event BlueprintInitializeAnimation()
   > Set Skel_Mesh(Skel_Mesh=GetOwnerSkeletalMeshComponent())
   n9 = Branch(Condition=OR Boolean!(A=Get bIsSkyDiving(), B=false, C=Get bIsSkydiveFloating(), D:bool=Get bIsSkydiveDiveMode(), E:bool, F:bool) -> (ReturnValue))
   > Set External_Force(External_Force:Vector="0,1000,1000") -> (Output_Get:Vector)
   > Set GravityOverride(GravityOverride="0, 0,500")
   > Set Linear_Acc_Scale(Linear_Acc_Scale:Vector="0.6,0.5,0.5") -> (Output_Get:Vector)
   > Set Linear_Vel_Scale(Linear_Vel_Scale:Vector="1,1,1") -> (Output_Get:Vector)
   Set External_Force(execute=@n9.else, External_Force="0, 0, 0")
   > Set GravityOverride(GravityOverride="0,0,-980")
   > Set Linear_Acc_Scale(Linear_Acc_Scale="0.6,0.5,0.5")
   > Set Linear_Vel_Scale(Linear_Vel_Scale="1,1,1")
   n10 = OR Boolean!(A=Get bIsCrouching(), B=Get bIsSprinting(), C=Get bIsPlayingMeleeAnim()) -> (ReturnValue)
   > n11 = Branch(Condition=AND Boolean(A=Get bIsMoving2D(), B=NOT Boolean(A=@n10)))
   > Set External_Force(External_Force="0, 0, 0")
   > Set GravityOverride(GravityOverride="0, 0,-980")
   > Set Linear_Acc_Scale(Linear_Acc_Scale="0.6,0.5,0.5")
   > Set Linear_Vel_Scale(Linear_Vel_Scale="1,1,1")
   Set External_Force(execute=@n11.else, External_Force="0, 0, 0")
   > Set GravityOverride(GravityOverride="0,0,-980")
   > Set Linear_Acc_Scale(Linear_Acc_Scale="0.6,0.5,0.5")
   > Set Linear_Vel_Scale(Linear_Vel_Scale="1,1,1")
   > n12 = Branch(Condition=Get bIsTacticalSprinting())
   > Set External_Force(External_Force="0,-2000,0")
   > Set GravityOverride(GravityOverride="0, 0,-350")
   > Set Linear_Acc_Scale(Linear_Acc_Scale="0.8,0.7,0.7")
   > Set Linear_Vel_Scale(Linear_Vel_Scale="3,3,3")
   Set External_Force(execute=@n12.else, External_Force="0, 0, 0")
   > Set GravityOverride(GravityOverride="0,0,-980")
   > Set Linear_Acc_Scale(Linear_Acc_Scale="0.6,0.5,0.5")
   > Set Linear_Vel_Scale(Linear_Vel_Scale="1,1,1")
   n13 = OR Boolean!(A=Get bIsTacticalSprinting(), B=Get bIsPlayingMeleeAnim()) -> (ReturnValue)
   > n14 = Branch(Condition=AND Boolean(A=Get bIsSprinting(), B=NOT Boolean(A=@n13)))
   > Set External_Force(External_Force="0, 0, 0")
   > Set GravityOverride(GravityOverride="0, 0,-980")
   > Set Linear_Acc_Scale(Linear_Acc_Scale="0.1,0.1,0.1")
   > Set Linear_Vel_Scale(Linear_Vel_Scale="1,1,1")
   Set External_Force(execute=@n14.else, External_Force="0, 0, 0")
   > Set GravityOverride(GravityOverride="0,0,-980")
   > Set Linear_Acc_Scale(Linear_Acc_Scale="0.6,0.5,0.5")
   > Set Linear_Vel_Scale(Linear_Vel_Scale="1,1,1")
   > n15 = Branch(Condition=Get bIsPlayingMeleeAnim())
   > Set External_Force(External_Force="0,-500, 0")
   > Set GravityOverride(GravityOverride="0, 0,-980")
   > Set Linear_Acc_Scale(Linear_Acc_Scale="1,1,1")
   > Set Linear_Vel_Scale(Linear_Vel_Scale="0.5,0.5,0.5")
   Set External_Force(execute=@n15.else, External_Force="0, 0, 0")
   > Set GravityOverride(GravityOverride="0,0,-980")
   > Set Linear_Acc_Scale(Linear_Acc_Scale="0.6,0.5,0.5")
   > Set Linear_Vel_Scale(Linear_Vel_Scale="1,1,1")
   TryGetPawnOwner()
   Branch()
  ]}
 UpdateBodyJoinTransforms {nodes [
   UpdateBodyJoinTransforms() {is=title K2Node_FunctionEntry}
   n1 = GetPartSkeletalMeshComponent()
   > Is Valid(InputObject=@n1)
   n2 = GetAttachParent!(self:Scene Component Object Reference=@n1) -> (ReturnValue:Scene Component Object Reference)
   > Is Valid(InputObject=@n2)
   > Set AttachParent(AttachParent:Scene Component Object Reference=@n2) -> (Output_Get:Scene Component Object Reference)
   > n3 = Sequence!(execute) -> (
    then_0
    then_1
    then_2
    then_3
    then_4:Exec
    then_5:Exec
    then_6:Exec
    then_7:Exec
    then_8:Exec
    then_9:Exec
    then_10:Exec
    then_11:Exec
    then_12:Exec
    then_13:Exec
    then_14:Exec
    then_15:Exec
    then_16:Exec
    then_17:Exec
   )
   n4 = Get AttachParent() -> (AttachParent:Scene Component Object Reference)
   n5 = BreakTransform(InTransform=GetSocketTransform(self=@n4, InSocketName=spine_05, TransformSpace=RTS_World))
   Set spine_05_tran(execute=@n3.then_10, spine_05_tran:Vector=@n5) -> (Output_Get:Vector)
   > Set spine_05_rot(spine_05_rot:Rotator=@n5.Rotation) -> (Output_Get:Rotator)
   n6 = BreakTransform(InTransform=GetSocketTransform(self=@n4, InSocketName=clavicle_r, TransformSpace=RTS_World))
   Set clavicle_r_tran(execute=@n3.then_11, clavicle_r_tran:Vector=@n6) -> (Output_Get:Vector)
   > Set clavicel_r_rot(clavicel_r_rot:Rotator=@n6.Rotation) -> (Output_Get:Rotator)
   n7 = BreakTransform(InTransform=GetSocketTransform(self=@n4, InSocketName=clavicle_l, TransformSpace=RTS_World))
   Set clavicle_l_tran(execute=@n3.then_12, clavicle_l_tran:Vector=@n7) -> (Output_Get:Vector)
   > Set clavicle_l_rot(clavicle_l_rot:Rotator=@n7.Rotation) -> (Output_Get:Rotator)
   n8 = BreakTransform(InTransform=GetSocketTransform(self=@n4, InSocketName=upperarm_r, TransformSpace=RTS_World))
   Set upperarm_r_tran(execute=@n3.then_13, upperarm_r_tran:Vector=@n8) -> (Output_Get:Vector)
   > Set upperarm_r_rot(upperarm_r_rot:Rotator=@n8.Rotation) -> (Output_Get:Rotator)
   n9 = BreakTransform(InTransform=GetSocketTransform(self=@n4, InSocketName=upperarm_l, TransformSpace=RTS_World))
   Set upperarm_l_tran(execute=@n3.then_14, upperarm_l_tran:Vector=@n9) -> (Output_Get:Vector)
   > Set upperarm_l_rot(upperarm_l_rot:Rotator=@n9.Rotation) -> (Output_Get:Rotator)
   n10 = BreakTransform(InTransform=GetSocketTransform(self=@n4, InSocketName=lowerarm_l, TransformSpace=RTS_World))
   Set lowerarm_l_tran(execute=@n3.then_15, lowerarm_l_tran:Vector=@n10) -> (Output_Get:Vector)
   > Set lowerarm_l_rot(lowerarm_l_rot:Rotator=@n10.Rotation) -> (Output_Get:Rotator)
   n11 = BreakTransform(InTransform=GetSocketTransform(self=@n4, InSocketName=lowerarm_r, TransformSpace=RTS_World))
   Set lowerarm_r_tran(execute=@n3.then_16, lowerarm_r_tran:Vector=@n11) -> (Output_Get:Vector)
   > Set lowerarm_r_rot(lowerarm_r_rot:Rotator=@n11.Rotation) -> (Output_Get:Rotator)
   n12 = BreakTransform(InTransform=GetSocketTransform(self=@n4, InSocketName=root, TransformSpace=RTS_World))
   Set root_trans(execute=@n3.then_17, root_trans:Vector=@n12) -> (Output_Get:Vector)
   > Set root_rot(root_rot:Rotator=@n12.Rotation) -> (Output_Get:Rotator)
   n13 = Get AttachParent()
   n14 = BreakTransform(InTransform=GetSocketTransform(self=@n13, InSocketName=spine_01, TransformSpace=RTS_World))
   Set spine_01_tran(execute=@n3.then_6, spine_01_tran:Vector=@n14) -> (Output_Get:Vector)
   > Set spine_01_rot(spine_01_rot:Rotator=@n14.Rotation) -> (Output_Get:Rotator)
   n15 = BreakTransform(InTransform=GetSocketTransform(self=@n13, InSocketName=spine_02, TransformSpace=RTS_World))
   Set spine_02_tran(execute=@n3.then_7, spine_02_tran:Vector=@n15) -> (Output_Get:Vector)
   > Set spine_02_rot(spine_02_rot:Rotator=@n15.Rotation) -> (Output_Get:Rotator)
   n16 = BreakTransform(InTransform=GetSocketTransform(self=@n13, InSocketName=spine_03, TransformSpace=RTS_World))
   Set spine_03_tran(execute=@n3.then_8, spine_03_tran:Vector=@n16) -> (Output_Get:Vector)
   > Set spine_03_rot(spine_03_rot:Rotator=@n16.Rotation) -> (Output_Get:Rotator)
   n17 = BreakTransform(InTransform=GetSocketTransform(self=@n13, InSocketName=spine_04, TransformSpace=RTS_World))
   Set spine_04_tran(execute=@n3.then_9, spine_04_tran:Vector=@n17) -> (Output_Get:Vector)
   > Set spine_04_rot(spine_04_rot:Rotator=@n17.Rotation) -> (Output_Get:Rotator)
   n18 = Get AttachParent()
   n19 = BreakTransform(InTransform=GetSocketTransform(self=@n18, InSocketName=pelvis, TransformSpace=RTS_World))
   Set pelvis_tran(execute=@n3.then_0, pelvis_tran:Vector=@n19) -> (Output_Get:Vector)
   > Set pelvis_rot(pelvis_rot:Rotator=@n19.Rotation) -> (Output_Get:Rotator)
   n20 = BreakTransform(InTransform=GetSocketTransform(self=@n18, InSocketName=thigh_l, TransformSpace=RTS_World))
   Set thigh_l_tran(execute=@n3.then_1, thigh_l_tran:Vector=@n20) -> (Output_Get:Vector)
   > Set thigh_l_rot(thigh_l_rot:Rotator=@n20.Rotation) -> (Output_Get:Rotator)
   n21 = BreakTransform(InTransform=GetSocketTransform(self=@n18, InSocketName=thigh_r, TransformSpace=RTS_World))
   Set thigh_r_tran(execute=@n3.then_2, thigh_r_tran:Vector=@n21) -> (Output_Get:Vector)
   > Set thigh_r_rot(thigh_r_rot:Rotator=@n21.Rotation) -> (Output_Get:Rotator)
   n22 = BreakTransform(InTransform=GetSocketTransform(self=@n18, InSocketName=calf_r, TransformSpace=RTS_World))
   Set calf_r_tran(execute=@n3.then_4, calf_r_tran:Vector=@n22) -> (Output_Get:Vector)
   > Set calf_r_rot(calf_r_rot:Rotator=@n22.Rotation) -> (Output_Get:Rotator)
   n23 = BreakTransform(InTransform=GetSocketTransform(self=@n18, InSocketName=calf_l, TransformSpace=RTS_World))
   Set calf_l_tran(execute=@n3.then_3, calf_l_tran:Vector=@n23) -> (Output_Get:Vector)
   > Set calf_l_rot(calf_l_rot:Rotator=@n23.Rotation) -> (Output_Get:Rotator)
   n24 = BreakTransform(InTransform=GetSocketTransform(self=@n18, InSocketName=root, TransformSpace=RTS_World))
   Set bodyRoot_tran(execute=@n3.then_5, bodyRoot_tran:Vector=@n24) -> (Output_Get:Vector)
   > Set bodyRoot_rot(bodyRoot_rot:Rotator=@n24.Rotation) -> (Output_Get:Rotator)
  ]}
]
variables [
 bodyRoot_rot {type=Rotator}
 bodyRoot_tran {type=Vector}
 calf_l_rot {type=Rotator}
 calf_l_tran {type=Vector}
 calf_r_rot {type=Rotator}
 calf_r_tran {type=Vector}
 clavicel_r_rot {type=Rotator}
 clavicle_l_rot {type=Rotator}
 clavicle_l_tran {type=Vector}
 clavicle_r_tran {type=Vector}
 Collider_Rotation {type=Rotator}
 Collider_Translation {type=Vector}
 External_Force {type=Vector}
 Gravity Override Params {
  type=GravityOverrideParamsStruct
  default {
   skydive_FloatingDown {gravityOverride=<0 -75 50>, windAmplitude=<75 -100 100>, windFrequency=<0.35 2 1>, windOffset=<0 -150 100>}
   skydive_GoingUp {gravityOverride=<0 -150 0>, windAmplitude=<100 -250 200>, windFrequency=<0.75 3.5 4>}
   skydive_Parachuting {gravityOverride=<0 -400 -720>, windAmplitude=<100 -250 200>, windFrequency=<0.75 3.5 4>}
  }
 }
 GravityOverride {type=Vector}
 IsFemale {type=boolean}
 IsMale {type=boolean}
 Linear_Acc_Scale {type=Vector}
 Linear_Vel_Scale {type=Vector}
 lowerarm_l_rot {type=Rotator}
 lowerarm_l_tran {type=Vector}
 lowerarm_r_rot {type=Rotator}
 lowerarm_r_tran {type=Vector}
 Pack_Translation {type=Vector}
 pelvis_rot {type=Rotator}
 pelvis_tran {type=Vector}
 root_rot {type=Rotator}
 root_trans {type=Vector}
 Rotator_01 {type=Rotator}
 Rotator_02 {type=Rotator}
 Skel_Mesh {type=SkeletalMeshComponent}
 spine_01_rot {type=Rotator}
 spine_01_tran {type=Vector}
 spine_02_rot {type=Rotator}
 spine_02_tran {type=Vector}
 spine_03_rot {type=Rotator}
 spine_03_tran {type=Vector}
 spine_04_rot {type=Rotator}
 spine_04_tran {type=Vector}
 spine_05_rot {type=Rotator}
 spine_05_tran {type=Vector}
 thigh_l_rot {type=Rotator}
 thigh_l_tran {type=Vector}
 thigh_r_rot {type=Rotator}
 thigh_r_tran {type=Vector}
 upperarm_l_rot {type=Rotator}
 upperarm_l_tran {type=Vector}
 upperarm_r_rot {type=Rotator}
 upperarm_r_tran {type=Vector}
]
